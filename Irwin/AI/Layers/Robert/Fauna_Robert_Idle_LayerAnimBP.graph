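# Animation layer blueprint for the Robert dinosaur that overrides the fauna idle layers.
# Fauna Idle plays the aggressive idle while bHasValidAimTarget is true, otherwise a random player picks evenly between the standard idle and a break variant.
# The bool switch uses inertialized blending, and the idle layer's input pose is left unused.
# Fauna IdleTurn passes its input pose straight through.
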AnimBlueprint /Irwin/AI/Layers/Robert/Fauna_Robert_Idle_LayerAnimBP
target_skeleton=/Irwin/AI/Base/Robert_Dinosaur_Invis_Skeleton
graphs [
 AnimGraph {nodes [Output Pose() {layerGroup=None}]}
 EventGraph {links [], nodes []}
 Fauna Idle {nodes [
   Input Pose() {name=InPose Idle}
   n1 = Random Sequence Player!() -> (Pose:Pose Link Structure) {
    is=title AnimGraphNode_RandomPlayer
    entries [
     {
      chanceToPlay=0.5
      maxLoopCount=0
      maxPlayRate=1
      minLoopCount=0
      minPlayRate=1
      sequence=/Irwin/AI/Predators/Robert/Animation/Robert_Idle
      blendIn {blendOption=HermiteCubic, blendTime=0.4, customCurve=null}
     }
     {
      chanceToPlay=0.5
      maxLoopCount=0
      maxPlayRate=1
      minLoopCount=0
      minPlayRate=1
      sequence=/Irwin/AI/Predators/Robert/Animation/Robert_Idle_Break1
      blendIn {blendOption=HermiteCubic, blendTime=0.4, customCurve=null}
     }
    ]
   }
   Output Pose(Result=Blend Poses by bool(
     BlendPose_0=Sequence Player 'Robert_Idle_Aggressive'() {sequence=/Irwin/AI/Predators/Robert/Animation/Robert_Idle_Aggressive}
     BlendPose_1=@n1
     BlendTime_0=0.4
     BlendTime_1=0.4
     bActiveValue=Get bHasValidAimTarget() -> (bHasValidAimTarget:bool)
    ) {blendType=Linear, transitionType=Inertialization}) {layerGroup=FaunaIdleOverride}
  ]}
 Fauna IdleTurn {nodes [Output Pose(Result=Input Pose() {name=InPose IdleTurn}) {layerGroup=FaunaIdleOverride}]}
]
variables []
