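# Animation layer Blueprint for the predator Robert, on the Robert_Dinosaur_Invis skeleton, that supplies the pose for the Fauna Jump layer.
# The Fauna Jump graph feeds a single Robert_Abduction sequence player straight into its output pose, so that animation serves as the jump.
# The main AnimGraph has no pose connected, and the asset has no event logic or variables.
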AnimBlueprint /Irwin/AI/Layers/Robert/Fauna_Robert_Jump_LayerAnimBP
target_skeleton=/Irwin/AI/Base/Robert_Dinosaur_Invis_Skeleton
graphs [
 AnimGraph {nodes [Output Pose() {layerGroup=None}]}
 EventGraph {links [], nodes []}
 Fauna Jump {nodes [Output Pose(Result=Sequence Player 'Robert_Abduction'() {sequence=/Irwin/AI/Predators/Robert/Animation/Robert_Abduction}) {layerGroup=None}]}
]
variables []
