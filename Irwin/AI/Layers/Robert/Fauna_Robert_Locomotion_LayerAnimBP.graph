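# Animation layer Blueprint for locomotion on the Robert fauna creature, targeting the Robert_Dinosaur_Invis skeleton.
# It defines two layers in the FaunaLocomotionOverride group, Fauna Locomotion and Fauna LocomotionTurn.
# Each layer passes its input pose straight to the output, so neither alters the animation.
# The main AnimGraph output is unconnected, and the event graph and variable list are empty.
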AnimBlueprint /Irwin/AI/Layers/Robert/Fauna_Robert_Locomotion_LayerAnimBP
target_skeleton=/Irwin/AI/Base/Robert_Dinosaur_Invis_Skeleton
graphs [
 AnimGraph {nodes [Output Pose() {layerGroup=None}]}
 EventGraph {links [], nodes []}
 Fauna Locomotion {nodes [Output Pose(Result=Input Pose() {name=InPose Locomotion}) {layerGroup=FaunaLocomotionOverride}]}
 Fauna LocomotionTurn {nodes [Output Pose(Result=Input Pose() {name=InPose LocomotionTurn}) {layerGroup=FaunaLocomotionOverride}]}
]
variables []
